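# Sound cue for Zordon's ArrowTube battle-encouragement voice lines, playing short rallying barks such as "We can win" and "Keep fighting".
# A Random node picks one of five equally weighted wave players, each holding a separate recorded line.
# The cue routes through the narrator VO sound class and sets a very large max distance.
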SoundCue /ToyFoxGameplay/Assets/Sound/VO/Zordon/VO_ArrowTube_Zordon_BattleEncourage
root=Random_0
nodes [
 {id=Random_0, children [WavePlayer_0,WavePlayer_1,WavePlayer_2,WavePlayer_3,WavePlayer_4], properties {weights [1,1,1,1,1]}}
 {id=WavePlayer_0, properties {soundWaveAssetPtr=/ToyFoxGameplay/Assets/Sound/VO/Zordon/Wavs/VO_ArrowTube_Zordon_WeCanWin_010}}
 {id=WavePlayer_1, properties {soundWaveAssetPtr=/ToyFoxGameplay/Assets/Sound/VO/Zordon/Wavs/VO_ArrowTube_Zordon_DontGiveUp_010}}
 {id=WavePlayer_2, properties {soundWaveAssetPtr=/ToyFoxGameplay/Assets/Sound/VO/Zordon/Wavs/VO_ArrowTube_Zordon_KeepFighting_010}}
 {id=WavePlayer_3, properties {soundWaveAssetPtr=/ToyFoxGameplay/Assets/Sound/VO/Zordon/Wavs/VO_ArrowTube_Zordon_MakeRitaRegret_010}}
 {id=WavePlayer_4, properties {soundWaveAssetPtr=/ToyFoxGameplay/Assets/Sound/VO/Zordon/Wavs/VO_ArrowTube_Zordon_StayFocused_010}}
]
properties {maxDistance=2097152, soundClassObject=/Game/Sounds/FortSoundClassesAndModes/Fort_VO_Narrator, volumeMultiplier=1}
