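# Sound cue for Zordon's ArrowTube voice-over lines, named for the pedestal spawn event, with several lines referencing the Megazord.
# A single random node with equal weights picks one of seven wave players, each holding a different Zordon voice line.
# The cue uses the Fort narrator VO sound class and a very large max distance.
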SoundCue /ToyFoxGameplay/Assets/Sound/VO/Zordon/VO_ArrowTube_Zordon_PedestalSpawn
root=Random_0
nodes [
 {
  id=Random_0
  children [WavePlayer_0,WavePlayer_1,WavePlayer_2,WavePlayer_3,WavePlayer_4,WavePlayer_5,WavePlayer_6]
  properties {weights [1,1,1,1,1,1,1]}
 }
 {id=WavePlayer_0, properties {soundWaveAssetPtr=/ToyFoxGameplay/Assets/Sound/VO/Zordon/Wavs/VO_ArrowTube_Zordon_MegazordWaiting_010}}
 {id=WavePlayer_1, properties {soundWaveAssetPtr=/ToyFoxGameplay/Assets/Sound/VO/Zordon/Wavs/VO_ArrowTube_Zordon_INeedA_010}}
 {id=WavePlayer_2, properties {soundWaveAssetPtr=/ToyFoxGameplay/Assets/Sound/VO/Zordon/Wavs/VO_ArrowTube_Zordon_RangersReport_010}}
 {id=WavePlayer_3, properties {soundWaveAssetPtr=/ToyFoxGameplay/Assets/Sound/VO/Zordon/Wavs/VO_ArrowTube_Zordon_WeHaveTo_010}}
 {id=WavePlayer_4, properties {soundWaveAssetPtr=/ToyFoxGameplay/Assets/Sound/VO/Zordon/Wavs/VO_ArrowTube_Zordon_UseMegazord_010}}
 {id=WavePlayer_5, properties {soundWaveAssetPtr=/ToyFoxGameplay/Assets/Sound/VO/Zordon/Wavs/VO_ArrowTube_Zordon_OnlyMegazord_010}}
 {id=WavePlayer_6, properties {soundWaveAssetPtr=/ToyFoxGameplay/Assets/Sound/VO/Zordon/Wavs/VO_ArrowTube_Zordon_WeNeedThe_010}}
]
properties {maxDistance=2097152, soundClassObject=/Game/Sounds/FortSoundClassesAndModes/Fort_VO_Narrator, volumeMultiplier=1}
